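# Physics asset for the LemonCartGranite two-handed pickaxe mesh, giving its handle and left/right shock joints simulation bodies.
# A capsule on the handle anchors the driver and wheel shock capsules on each side through constraints with locked twist.
# Each ExtraMid shock capsule hangs off its driver with swing and twist locked.
# The two wheel constraints are not mirrored: only the left locks swing1, and their swing2 limit angles differ.
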PhysicsAsset /BRCosmetics/Weapons/FORT_Melee/Pickaxe_LemonCartGranite/Meshes/LemonCartGranite_Axe_FE_Physics
bodies {
 Melee_TwoHanded_Handle [{shapeType=Capsule, length=70.1275, radius=11.3606, center=<-0.1978 -1.1378 24.216>, rotation=<rot 0.1009 -90 -1.3167>}]
 l_Wheel_shock_02_jnt [{shapeType=Capsule, length=10, radius=5, shapeName=l_Wheel_shock_02_jnt_capsule, rotation=<rot 65 0 0>}]
 l_driver_shock_01_jnt [{shapeType=Capsule, length=15, radius=5, shapeName=l_driver_shock_01_jnt_capsule, center=<7.3704 0 0>, rotation=<rot 90 0 0>}]
 l_dyn_ExtraMid_shock_jnt [{shapeType=Capsule, length=5.1589, radius=2.7655, shapeName=l_dyn_ExtraMid_shock_jnt_capsule, rotation=<rot 90 0 0>}]
 r_Wheel_shock_02_jnt [{shapeType=Capsule, length=10, radius=5, shapeName=l_Wheel_shock_02_jnt_capsule, rotation=<rot 65 0 -180>}]
 r_driver_shock_01_jnt [{shapeType=Capsule, length=15, radius=5, shapeName=l_driver_shock_01_jnt_capsule, center=<-7.3704 0 0>, rotation=<rot 90 0 180>}]
 r_dyn_ExtraMid_shock_jnt [{shapeType=Capsule, length=5.1589, radius=2.7655, shapeName=r_dyn_ExtraMid_shock_jnt_capsule, rotation=<rot 90 0 0>}]
}
constraints [
 {
  bone1Name=l_driver_shock_01_jnt
  bone2Name=Melee_TwoHanded_Handle
  swing1LimitDegrees=5
  swing2LimitDegrees=3
  twistLimitDegrees=45
  twistMotion=Locked
 }
 {
  bone1Name=l_Wheel_shock_02_jnt
  bone2Name=Melee_TwoHanded_Handle
  swing1LimitDegrees=2
  swing1Motion=Locked
  swing2LimitDegrees=35
  twistLimitDegrees=3
  twistMotion=Locked
 }
 {
  bone1Name=r_driver_shock_01_jnt
  bone2Name=Melee_TwoHanded_Handle
  swing1LimitDegrees=5
  swing2LimitDegrees=3
  twistLimitDegrees=45
  twistMotion=Locked
 }
 {
  bone1Name=l_dyn_ExtraMid_shock_jnt
  bone2Name=l_driver_shock_01_jnt
  swing1LimitDegrees=45
  swing1Motion=Locked
  swing2LimitDegrees=45
  swing2Motion=Locked
  twistLimitDegrees=45
  twistMotion=Locked
 }
 {
  bone1Name=r_dyn_ExtraMid_shock_jnt
  bone2Name=r_driver_shock_01_jnt
  swing1LimitDegrees=45
  swing1Motion=Locked
  swing2LimitDegrees=45
  swing2Motion=Locked
  twistLimitDegrees=45
  twistMotion=Locked
 }
 {
  bone1Name=r_Wheel_shock_02_jnt
  bone2Name=Melee_TwoHanded_Handle
  swing1LimitDegrees=2
  swing2LimitDegrees=25
  twistLimitDegrees=45
  twistMotion=Locked
 }
]
