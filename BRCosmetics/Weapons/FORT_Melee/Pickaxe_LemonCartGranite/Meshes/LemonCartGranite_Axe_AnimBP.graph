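# Animation Blueprint for the LemonCartGranite pickaxe, adding secondary physics motion to its shocks, wheels and dog tag.
# In game, copied mesh pose runs through RigidBody and shock constraints into a cache, then AnimDynamics and ZMovement-weighted bone-driven wheel rotation.
# Front end instead uses separate RigidBody physics, layered with cached shock bones.
# Event graph updates ZMovement and retunes gravity and inertia while sprint-jumping; Look At and Flap chains are unconnected.
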AnimBlueprint /BRCosmetics/Weapons/FORT_Melee/Pickaxe_LemonCartGranite/Meshes/LemonCartGranite_Axe_AnimBP
target_skeleton=&/LemonCartGranite_Axe_Skeleton
graphs [
 AnimGraph {nodes [
   n1 = Look At - Bone: l_driven_shock_01_jnt(Alpha=1) {
    alphaBoolBlend {blendOption=Linear}
    boneToModify {boneName=l_driven_shock_01_jnt}
    lookAtLocation {x=0, z=-2}
    lookAtTarget {boneReference {boneName=l_shock_02_AimConst_jnt}}
    lookAt_Axis {axis=<1 0>}
    lookUp_Axis {axis=<0 1>}
   }
   Look At - Bone: r_driven_shock_01_jnt(ComponentPose=@n1, Alpha=1) {
    alphaBoolBlend {blendOption=Linear}
    boneToModify {boneName=r_driven_shock_01_jnt}
    lookAtLocation {x=0, z=1}
    lookAtTarget {boneReference {boneName=r_shock_02_AimConst_jnt}}
    lookAt_Axis {axis=<-1 0>}
    lookUp_Axis {axis=<0 1>}
   }
   n2 = Anim Dynamics - Bone: l_dyn_dangling_dogTtag_jnt(ComponentPose=Local To Component(LocalPose=Use cached pose 'preBaseSimCache'()), Alpha=1, ExternalForce="0,0,0") {
    bLinearSpring=true
    bOverrideLinearDamping=true
    bUseGravityOverride=true
    linearDampingOverride=0.999
    linearSpringConstant=320
    simulationSpace=BoneRelative
    alphaBoolBlend {blendOption=Linear}
    boundBone {boneName=l_dyn_dangling_dogTtag_jnt}
    componentAppliedLinearAccClamp=<2000 2000 2000>
    gravityOverride {z=-980}
    physicsBodyDefinitions [{
      boundBone {boneName=l_dyn_dangling_dogTtag_jnt}
      boxExtents=<2.5 2.5 2.5>
      constraintSetup {angularLimitsMax=<0 0 0>, angularLimitsMin=<0 0 0>, angularTarget=<0 0 0>, linearAxesMax=<1 0.7 2>, linearAxesMin=<0 -0.7 -2>}
      localJointOffset=<1 1 0>
     }]
    relativeSpaceBone {boneName=LeftDogPickAxe_jnts}
   }
   n3 = Get ZMovement() -> (ZMovement:float)
   n4 = "r_Wheel_shock_02_jnt.rotateY = r_driver_shock_01_jnt.rotateZ * 2.5 - Bone Driven Controller"!(ComponentPose:Component Space Pose Link Structure=@n2, Alpha:float=@n3) -> (Pose:Component Space Pose Link Structure) {
    is=title AnimGraphNode_BoneDrivenController
    bAffectTargetRotationY=true
    modificationMode=ReplaceComponent
    multiplier=2.5
    sourceComponent=RotationZ
    alphaBoolBlend {blendOption=Linear}
    alphaScaleBiasClamp {bInterpResult=true, bias=1, interpSpeedDecreasing=80, interpSpeedIncreasing=25}
    sourceBone {boneName=r_driver_shock_01_jnt}
    targetBone {boneName=r_Wheel_shock_02_jnt}
   }
   n5 = "l_Wheel_shock_02_jnt.rotateY = l_driver_shock_01_jnt.rotateZ * 2.5 - Bone Driven Controller"!(ComponentPose:Component Space Pose Link Structure=@n4, Alpha:float=@n3) -> (Pose:Component Space Pose Link Structure) {
    is=title AnimGraphNode_BoneDrivenController
    bAffectTargetRotationY=true
    modificationMode=ReplaceComponent
    multiplier=2.5
    sourceComponent=RotationZ
    alphaBoolBlend {blendOption=Linear}
    alphaScaleBiasClamp {bInterpResult=true, bias=1, interpSpeedDecreasing=80, interpSpeedIncreasing=25}
    sourceBone {boneName=l_driver_shock_01_jnt}
    targetBone {boneName=l_Wheel_shock_02_jnt}
   }
   n6 = "r_Wheel_shock_02_jnt.rotateZ = r_driver_shock_01_jnt.rotateZ * 1.5 - Bone Driven Controller"!(ComponentPose:Component Space Pose Link Structure=@n5, Alpha:float=@n3) -> (Pose:Component Space Pose Link Structure) {
    is=title AnimGraphNode_BoneDrivenController
    bAffectTargetRotationZ=true
    modificationMode=ReplaceComponent
    multiplier=1.5
    sourceComponent=RotationZ
    alphaBoolBlend {blendOption=Linear}
    alphaScaleBiasClamp {bInterpResult=true, bias=1, interpSpeedDecreasing=80, interpSpeedIncreasing=25}
    sourceBone {boneName=r_driver_shock_01_jnt}
    targetBone {boneName=r_Wheel_shock_02_jnt}
   }
   n7 = "l_Wheel_shock_02_jnt.rotateZ = l_driver_shock_01_jnt.rotateZ * 1.5 - Bone Driven Controller"!(ComponentPose:Component Space Pose Link Structure=@n6, Alpha:float=@n3) -> (Pose:Component Space Pose Link Structure) {
    is=title AnimGraphNode_BoneDrivenController
    bAffectTargetRotationZ=true
    modificationMode=ReplaceComponent
    multiplier=1.5
    sourceComponent=RotationZ
    alphaBoolBlend {blendOption=Linear}
    alphaScaleBiasClamp {bInterpResult=true, bias=1, interpSpeedDecreasing=80, interpSpeedIncreasing=25}
    sourceBone {boneName=l_driver_shock_01_jnt}
    targetBone {boneName=l_Wheel_shock_02_jnt}
   }
   n8 = RigidBody!(
    ComponentPose=Local To Component(LocalPose=Input Pose() {name=InPose_0})
    Alpha
    OverrideWorldGravity:Vector="0,0,-760"
    ExternalForce="1,1,1"
   ) -> (Pose) {
    bOverrideWorldGravity=true
    overridePhysicsAsset=&/LemonCartGranite_Axe_FE_Physics
    alphaBoolBlend {blendOption=Linear}
    alphaScaleBiasClamp {bInterpResult=true}
    baseBoneRef {boneName=Melee_TwoHanded_Handle}
    componentAppliedLinearAccClamp=<1000 1000 1000>
    componentLinearAccScale=<0.35 0.35 0.35>
    componentLinearVelScale=<1 1 1>
    overrideWorldGravity {z=-300}
    simSpaceSettings {maxAngularAcceleration=1000, maxAngularVelocity=300, maxLinearAcceleration=10, maxLinearVelocity=50, velocityScaleZ=0.8, worldAlpha=1}
   }
   n9 = Layered blend per bone(BasePose=Component To Local(ComponentPose=@n8), BlendPoses_0=Use cached pose 'preBaseSimCache'()) {layerSetup [{branchFilters [{boneName=l_shock_02_AimConst_jnt}, {boneName=r_shock_02_AimConst_jnt}, {boneName=l_driven_shock_02_jnt}, {boneName=r_driven_shock_02_jnt}]}]}
   Output Pose(Result=Blend Poses by bool(BlendPose_0=@n9, BlendPose_1=Component To Local(ComponentPose=@n7), bActiveValue=Get bIsInFrontEnd()) {blendType=Linear}) {layerGroup=None}
   n10 = RigidBody!(
    ComponentPose=Local To Component(LocalPose=Copy Pose From Mesh())
    Alpha
    OverrideWorldGravity=Get Override World Gravity() -> (Override World Gravity:Vector)
    ExternalForce
    ComponentLinearAccScale:Vector=Get Component Linear Acc Scale() -> (Component Linear Acc Scale:Vector)
    ComponentLinearVelScale:Vector=Get Component Linear Vel Scale() -> (Component Linear Vel Scale:Vector)
    bTransferBoneVelocities:bool=False
   ) -> (Pose) {
    bOverrideWorldGravity=true
    bTransferBoneVelocities=true
    overridePhysicsAsset=&/LemonCartGranite_Axe_Physics
    alphaBoolBlend {blendOption=Linear}
    alphaScaleBiasClamp {bInterpResult=true}
    componentAppliedLinearAccClamp=<1000 1000 1000>
    componentLinearAccScale=<0.1 0.1 0.1>
    componentLinearVelScale=<0.3 0.5 0.3>
    overrideWorldGravity {z=-200}
    simSpaceSettings {maxAngularAcceleration=1300, maxAngularVelocity=30, maxLinearAcceleration=700, maxLinearVelocity=25, worldAlpha=1}
   }
   n11 = "Constraint - l_driven_shock_02_jnt (T)"(ComponentPose=@n10, Alpha=1, ConstraintWeights_0=1) {
    alphaBoolBlend {blendOption=Linear}
    boneToModify {boneName=l_driven_shock_02_jnt}
    constraintSetup [{transformType=Translation, targetBone {boneName=l_shock_02_AimConst_jnt}}]
    constraintWeights [1]
   }
   n12 = "Constraint - r_driven_shock_02_jnt (T)"(ComponentPose=@n11, Alpha=1, ConstraintWeights_0=1) {
    alphaBoolBlend {blendOption=Linear}
    boneToModify {boneName=r_driven_shock_02_jnt}
    constraintSetup [{transformType=Translation, targetBone {boneName=r_shock_02_AimConst_jnt}}]
    constraintWeights [1]
   }
   Save cached pose 'preBaseSimCache'(Pose=Component To Local(ComponentPose=@n12)) {cacheName=preBaseSimCache}
   n13 = Flap Node - Bone: l_dyn_ExtraMid_shock_jnt!(ComponentPose:Component Space Pose Link Structure, Alpha:float=1) -> (Pose:Component Space Pose Link Structure) {
    is=title FortAnimGraphNode_Flap
    targetClampMax=0.5
    targetComponent=ScaleX
    alphaBoolBlend {blendOption=Linear}
    alphaScaleBiasClamp {bInterpResult=true}
    sourceDrives [{
      bUseRange=false
      drivingCurve=null
      multiplier=-0.65
      rangeMax=1
      rangeMin=-1
      remappedMax=1
      remappedMin=0
      sourceComponent=RotationZ
      useQuaternion=true
      sourceBone {boneName=l_driver_shock_01_jnt}
     }]
    targetBone {boneName=l_dyn_ExtraMid_shock_jnt}
   }
   n14 = Flap Node - Bone: l_dyn_ExtraUpper_shock_jnt!(ComponentPose:Component Space Pose Link Structure=@n13, Alpha:float=1) -> (Pose:Component Space Pose Link Structure) {
    is=title FortAnimGraphNode_Flap
    targetClampMax=0.5
    targetComponent=ScaleX
    alphaBoolBlend {blendOption=Linear}
    alphaScaleBiasClamp {bInterpResult=true}
    sourceDrives [{
      bUseRange=false
      drivingCurve=null
      multiplier=1
      rangeMax=1
      rangeMin=-1
      remappedMax=1
      remappedMin=0
      sourceComponent=RotationZ
      useQuaternion=true
      sourceBone {boneName=l_driver_shock_01_jnt}
     }]
    targetBone {boneName=l_dyn_ExtraUpper_shock_jnt}
   }
   n15 = Flap Node - Bone: l_dyn_ExtraLower_shock_jnt!(ComponentPose:Component Space Pose Link Structure=@n14, Alpha:float=1) -> (Pose:Component Space Pose Link Structure) {
    is=title FortAnimGraphNode_Flap
    targetClampMax=0.5
    targetComponent=ScaleX
    alphaBoolBlend {blendOption=Linear}
    alphaScaleBiasClamp {bInterpResult=true}
    sourceDrives [{
      bUseRange=false
      drivingCurve=null
      multiplier=1
      rangeMax=1
      rangeMin=-1
      remappedMax=1
      remappedMin=0
      sourceComponent=RotationZ
      useQuaternion=true
      sourceBone {boneName=l_driver_shock_01_jnt}
     }]
    targetBone {boneName=l_dyn_ExtraLower_shock_jnt}
   }
   n16 = Flap Node - Bone: r_dyn_ExtraMid_shock_jnt!(ComponentPose:Component Space Pose Link Structure=@n15, Alpha:float=1) -> (Pose:Component Space Pose Link Structure) {
    is=title FortAnimGraphNode_Flap
    targetClampMax=0.5
    targetComponent=ScaleX
    alphaBoolBlend {blendOption=Linear}
    alphaScaleBiasClamp {bInterpResult=true}
    sourceDrives [{
      bUseRange=false
      drivingCurve=null
      multiplier=-0.65
      rangeMax=1
      rangeMin=-1
      remappedMax=1
      remappedMin=0
      sourceComponent=RotationZ
      useQuaternion=true
      sourceBone {boneName=l_driver_shock_01_jnt}
     }]
    targetBone {boneName=r_dyn_ExtraMid_shock_jnt}
   }
   n17 = Flap Node - Bone: r_dyn_ExtraUpper_shock_jnt!(ComponentPose:Component Space Pose Link Structure=@n16, Alpha:float=1) -> (Pose:Component Space Pose Link Structure) {
    is=title FortAnimGraphNode_Flap
    targetClampMax=0.5
    targetComponent=ScaleX
    alphaBoolBlend {blendOption=Linear}
    alphaScaleBiasClamp {bInterpResult=true}
    sourceDrives [{
      bUseRange=false
      drivingCurve=null
      multiplier=1
      rangeMax=1
      rangeMin=-1
      remappedMax=1
      remappedMin=0
      sourceComponent=RotationZ
      useQuaternion=true
      sourceBone {boneName=l_driver_shock_01_jnt}
     }]
    targetBone {boneName=r_dyn_ExtraUpper_shock_jnt}
   }
   Flap Node - Bone: r_dyn_ExtraLower_shock_jnt!(ComponentPose:Component Space Pose Link Structure=@n17, Alpha:float=1) -> (Pose:Component Space Pose Link Structure) {
    is=title FortAnimGraphNode_Flap
    targetClampMax=0.5
    targetComponent=ScaleX
    alphaBoolBlend {blendOption=Linear}
    alphaScaleBiasClamp {bInterpResult=true}
    sourceDrives [{
      bUseRange=false
      drivingCurve=null
      multiplier=1
      rangeMax=1
      rangeMin=-1
      remappedMax=1
      remappedMin=0
      sourceComponent=RotationZ
      useQuaternion=true
      sourceBone {boneName=l_driver_shock_01_jnt}
     }]
    targetBone {boneName=r_dyn_ExtraLower_shock_jnt}
   }
   n18 = Flap Node - Bone: l_driven_shock_02_jnt!(ComponentPose:Component Space Pose Link Structure, Alpha:float=1) -> (Pose:Component Space Pose Link Structure) {
    is=title FortAnimGraphNode_Flap
    targetClampMax=10
    targetComponent=TranslationX
    alphaBoolBlend {blendOption=Linear}
    alphaScaleBiasClamp {interpSpeedDecreasing=5, interpSpeedIncreasing=5}
    sourceDrives [{
      bUseRange=false
      drivingCurve=null
      multiplier=23
      rangeMax=1
      rangeMin=-1
      remappedMax=1
      remappedMin=0
      sourceComponent=RotationZ
      useQuaternion=true
      sourceBone {boneName=l_driver_shock_01_jnt}
     }]
    targetBone {boneName=l_driven_shock_02_jnt}
   }
   Flap Node - Bone: r_driven_shock_02_jnt!(ComponentPose:Component Space Pose Link Structure=@n18, Alpha:float=1) -> (Pose:Component Space Pose Link Structure) {
    is=title FortAnimGraphNode_Flap
    targetClampMax=10
    targetComponent=TranslationX
    alphaBoolBlend {blendOption=Linear}
    alphaScaleBiasClamp {interpSpeedDecreasing=5, interpSpeedIncreasing=5}
    sourceDrives [{
      bUseRange=false
      drivingCurve=null
      multiplier=-10
      rangeMax=1
      rangeMin=-1
      remappedMax=1
      remappedMin=0
      sourceComponent=RotationZ
      useQuaternion=true
      sourceBone {boneName=r_driver_shock_01_jnt}
     }]
    targetBone {boneName=r_driven_shock_02_jnt}
   }
  ]}
 EventGraph {nodes [
   n1 = Event BlueprintUpdateAnimation()
   > Set Delta Time X(Delta Time X=@n1.DeltaTimeX)
   > n2 = Sequence!(execute) -> (then_0, then_1, then_2:Exec, then_3:Exec)
   > n3 = Gravity Override State Function DEPRECATED(CharacterPartAnimInstance=Self-Reference(), GameTimeInSeconds=GetGameTimeInSeconds(), GravityOverrideParams=Get Gravity Override Params())
   > Set Gravity Override(Gravity Override=@n3)
   n4 = PlayerMovement(execute=@n2.then_1, AnimInstance=Self-Reference(), XDotProductCoefficient=1, YDotProductCoefficient=1, ZDotProductCoefficient=1)
   > Set ZMovement(ZMovement:float="Max (Float)"(A=@n4.ZMovement, B=0)) -> (Output_Get:float)
   > n5 = Branch(Condition=AND Boolean(A=Get bIsSprinting(), B=Get bIsJumping()))
   n6 = Set Override World Gravity(execute=@n5.else, Override World Gravity:Vector="0,0,-200") -> (Output_Get:Vector)
   n7 = vector + vector(A=@n6, B="0,0,-100")
   Set Override World Gravity(
    execute=@n5.then
    Override World Gravity=TInterpTo!(
     Current:"Transform (by ref)"="To Transform (Vector)"!(InLocation:Vector=@n6) -> (ReturnValue:Transform)
     Target:"Transform (by ref)"="To Transform (Vector)"(InLocation=@n7)
     DeltaTime:float=Get Delta Time X()
     InterpSpeed:float=10
    ) -> (ReturnValue_Location:Vector, ReturnValue_Rotation:Rotator, ReturnValue_Scale:Vector)
   )
   > Set Component Linear Acc Scale(Component Linear Acc Scale:Vector="0.12,0.2,0.12") -> (Output_Get:Vector)
   > Set Component Linear Vel Scale(Component Linear Vel Scale:Vector="1,1,1") -> (Output_Get:Vector)
   Event OnEndJumping()
   > Do Once(Reset=@n8.then_1)
   > Set JumpingEnd(JumpingEnd:bool=true) -> (Output_Get:bool)
   > Delay(Duration=0.98)
   > Set JumpingEnd(JumpingEnd=false)
   Event OnBeginJumping()
   > n8 = Sequence!(execute) -> (then_0, then_1)
   > Set JumpingEnd(JumpingEnd=false)
   TryGetPawnOwner()
   Set Component Linear Vel Scale(execute=@n9.then, Component Linear Vel Scale="0.3,0.5,0.3")
   n9 = Set Component Linear Acc Scale(execute=@n6.then, Component Linear Acc Scale="0.1,0.1,0.1")
  ]}
]
variables [
 compensationValue {type=Vector}
 Component Linear Acc Scale {type=Vector, default=<0.1 0.1 0.1>}
 Component Linear Vel Scale {type=Vector, default=<0.3 0.5 0.3>}
 CurrentValueY {type=Vector}
 Delta Time X {type=number}
 Gravity Override {type=Vector}
 Gravity Override Params {type=GravityOverrideParamsStruct}
 JumpingEnd {type=boolean}
 Override World Gravity {type=Vector, default=<0 0 -200>}
 ZMovement {type=number}
]
